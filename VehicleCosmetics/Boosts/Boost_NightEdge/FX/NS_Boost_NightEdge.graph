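# Niagara system for the NightEdge vehicle boost cosmetic, drawing ribbon trails from the car's boost ports.
# One emitter spawns a ribbon per filtered socket at a quality-handled rate, offset by the car offset and pushed backward while inheriting source movement.
# Four ribbon renderers share the particles, with extra R2–R4 attributes driving layered widths, colors and alpha over lifetime.
# The painted-variant override module is disabled.
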NiagaraSystem /VehicleCosmetics/Boosts/Boost_NightEdge/FX/NS_Boost_NightEdge
dependencies {
 usedDataInterfaces [&NiagaraDataInterfaceCurve, &NiagaraDataInterfaceSkeletalMesh]
 usedDynamicInputs []
 usedModules [
  &self:SystemSpawnScript.SystemScriptSource.SystemScriptGraph.NiagaraNodeAssignment_0.SetVariables_#1
  &SystemState
 ]
}
emitters {Ribbons {
  emitterSpawnScript [/SetVariables(Emitter.R2Width=80) {isSetParameters=true}]
  emitterUpdateScript [
   &EmitterState()
   &VC_FilteredSocketCount()
   &VC_RateQualityHandler(Default Rate=90)
   &VC_SpawnRibbon_v1(Number of Ribbons=$Emitter.PortCount, SpawnRate=$Emitter.Spawn Rate)
  ]
  particleSpawnScript [
   &VC_PaintedVariantsOverride() {enabled=false}
   &VC_InitializeRibbon_v1(
    Custom Offset=true
    Enable Skeletal Position=true
    Lifetime=0.6
    Local Vector Offset=&AddVectorToPosition(Position=<90 0 0>, Vector=$User.DTCarOffset)
    Offset Amount=1
    Ribbon Width:NiagaraFloat=&Multiply_FloatByInt(Float=9, Integer=10)
   )
   &AddVelocity(Velocity=<-2000 0 0>) {module=AddVelocity001}
  ]
  particleUpdateScript [
   &ParticleState()
   &InheritSourceMovement(
    Applied Position Scale=&Multiply_VectorByFloat(
     Vector=<1 1 1>
     Float=&FloatFromCurve(FloatCurve=DataInterface() {propertyValues {Curve {keys [
          {interpMode=RCIM_Cubic, leaveTangent=-0.0051, tangentMode=RCTM_Break, time=0, value=1}
          {arriveTangent=-0.0476, interpMode=RCIM_Cubic, tangentMode=RCTM_Break, time=1, value=0.9}
         ]}}})
    )
    ApplyForce=false
    ApplyPosition=true
    ApplyPositionFalloff=false
   )
   &SolveForcesAndVelocity(Clamp Velocity=false, Limit Acceleration=false)
   &VC_UpdateRibbon_v1(Enable Skeletal Position=true)
   /SetVariables(
    Particles.R2 Color:LinearColor=&VC_DynamicScaleColor(
     Color to Scale:LinearColor=<rgba 1 1 1 1>
     RGB Scale:Vector3f=&VectorFromFloat(Value=0.2)
     Alpha Scale:NiagaraFloat=&FloatFromCurve(FloatCurve=DataInterface() {propertyValues {Curve {
         defaultValue=1
         keys [
          {arriveTangentWeight=0.1317, interpMode=RCIM_Cubic, leaveTangentWeight=0.1317, time=0.1, value=0}
          {
           arriveTangent=4.9502
           arriveTangentWeight=0.107
           interpMode=RCIM_Cubic
           leaveTangent=4.9502
           leaveTangentWeight=0.107
           tangentMode=RCTM_User
           tangentWeightMode=RCTWM_WeightedBoth
           time=0.2
           value=1
          }
          <at 0.4 1 RCIM_Cubic>
          {tangentMode=RCTM_User, time=1, value=1}
         ]
        }}})
    )
    Particles.R2 Width:NiagaraFloat=&Multiply_Float(
     A=$Emitter.R2Width
     B=&FloatFromCurve(FloatCurve=DataInterface() {propertyValues {Curve {keys [<at 0 0.6>,<at 0.2 1>,<at 0.6 1>,<at 1 0.5>]}}}, Scale Curve=0.8)
    )
    Particles.R3Width:NiagaraFloat=&Multiply_Float(
     A=$Emitter.R2Width
     B=&FloatFromCurve(FloatCurve=DataInterface() {propertyValues {Curve {keys [<at 0.6 1>,<at 0.8 0.7875>,<at 1 0.21>]}}}, Scale Curve=0.5)
    )
    Particles.R4 Color:LinearColor=&VC_DynamicScaleColor(
     Color to Scale:LinearColor=<rgba 1 1 1 1>
     RGB Scale:Vector3f=&VectorFromFloat(Value=1)
     Alpha Scale:NiagaraFloat=&FloatFromCurve(FloatCurve=DataInterface() {propertyValues {Curve {
         defaultValue=1
         keys [
          {arriveTangentWeight=0.1317, interpMode=RCIM_Cubic, leaveTangentWeight=0.1317, time=0.15, value=0}
          {
           arriveTangentWeight=0.0667
           interpMode=RCIM_Cubic
           leaveTangentWeight=0.0667
           tangentMode=RCTM_User
           tangentWeightMode=RCTWM_WeightedBoth
           time=0.2
           value=1
          }
          {arriveTangentWeight=0.0667, interpMode=RCIM_Cubic, leaveTangentWeight=0.1333, tangentMode=RCTM_User, time=0.4, value=1}
          {arriveTangentWeight=0.1334, interpMode=RCIM_Cubic, leaveTangentWeight=0.0667, tangentMode=RCTM_User, time=0.7, value=1}
          {tangentMode=RCTM_User, time=0.85, value=0}
         ]
        }}})
    )
    Particles.R4 Pos:NiagaraPosition=&AddVectorToPosition(Position=$Particles.Position, Vector=<0 0 0>)
    Particles.R4 Width:NiagaraFloat=&Multiply_Float(
     A=$Emitter.R2Width
     B=&FloatFromCurve(FloatCurve=DataInterface() {propertyValues {Curve {keys [<at 0 0.6>,<at 0.2 1>,<at 0.6 1>,<at 1 0.6>]}}}, Scale Curve=0.84)
    )
   ) {isSetParameters=true}
   &ScaleColor(
    Scale Alpha=&FloatFromCurve(
     FloatCurve=DataInterface() {propertyValues {Curve {keys [
         {
          arriveTangent=-0.0633
          arriveTangentWeight=0.1695
          interpMode=RCIM_Cubic
          leaveTangent=-0.0633
          leaveTangentWeight=0.1695
          tangentMode=RCTM_User
          time=0
          value=0
         }
         {
          arriveTangent=0.0087
          arriveTangentWeight=0.1036
          interpMode=RCIM_Cubic
          leaveTangent=0.0087
          leaveTangentWeight=0.1036
          tangentMode=RCTM_User
          tangentWeightMode=RCTWM_WeightedBoth
          time=0.25
          value=0.96
         }
         {
          arriveTangent=-0.1369
          arriveTangentWeight=0.1353
          interpMode=RCIM_Cubic
          leaveTangent=-0.1369
          leaveTangentWeight=0.1353
          tangentMode=RCTM_User
          tangentWeightMode=RCTWM_WeightedBoth
          time=0.72
          value=0.94
         }
         {arriveTangent=0.0718, arriveTangentWeight=0.1, tangentMode=RCTM_User, time=1, value=0}
        ]}}}
     Scale Curve=8
    )
    ScaleA=true
    ScaleRGB=true
   )
   &ScaleRibbonWidth(Ribbon Width Scale=&FloatFromCurve(FloatCurve=DataInterface() {propertyValues {Curve {keys [<at 0 0.3>,<at 0.2 0.7>,<at 0.5 0.9>,<at 1 0.7>]}}}))
   &DynamicMaterialParameters(
    Index 0 Param 1=&FloatFromCurve(FloatCurve=DataInterface() {propertyValues {Curve {keys [
         {
          arriveTangent=-0.0633
          arriveTangentWeight=0.1695
          interpMode=RCIM_Cubic
          leaveTangent=-0.0633
          leaveTangentWeight=0.1695
          tangentMode=RCTM_User
          time=0
          value=0.5
         }
         {
          arriveTangent=0.0087
          arriveTangentWeight=0.1036
          interpMode=RCIM_Cubic
          leaveTangent=0.0087
          leaveTangentWeight=0.1036
          tangentMode=RCTM_User
          tangentWeightMode=RCTWM_WeightedBoth
          time=0.25
          value=0.96
         }
         {
          arriveTangent=-0.1369
          arriveTangentWeight=0.1353
          interpMode=RCIM_Cubic
          leaveTangent=-0.1369
          leaveTangentWeight=0.1353
          tangentMode=RCTM_User
          tangentWeightMode=RCTWM_WeightedBoth
          time=0.72
          value=0.94
         }
         {arriveTangent=0.0718, arriveTangentWeight=0.1, tangentMode=RCTM_User, time=1, value=0.55}
        ]}}})
    Module.Index0Float0Write=true
    Module.Index0Float1Write=true
    Module.Index0Float2Write=true
    Module.Index0Float3Write=true
   )
  ]
  rendererClasses [&NiagaraRibbonRendererProperties]
  renderers [
   {rendererClass=&NiagaraRibbonRendererProperties}
   {rendererClass=&NiagaraRibbonRendererProperties, rendererIndex=1}
   {rendererClass=&NiagaraRibbonRendererProperties, rendererIndex=2}
   {rendererClass=&NiagaraRibbonRendererProperties, rendererIndex=3}
  ]
 }}
summary {
 systemName=NS_Boost_NightEdge
 emitters [{emitterName=Ribbons, rendererClasses [&NiagaraRibbonRendererProperties]}]
 userVariables [
  User.bIsFrontendPreview {type=NiagaraBool}
  User.bIsPainted {type=NiagaraBool}
  User.ColorVariantID {type=NiagaraInt32}
  User.DTCarOffset {type=Vector3f, description="", defaultValue {struct=Vector3f, value=<0 0 0>}}
  User.PaintedColor {type=LinearColor, description="", defaultValue {struct=LinearColor, value=<rgba 1 1 1 1>}}
  User.PaintedScalar {type=NiagaraFloat, defaultValue=1, description=""}
  User.PortCount {type=NiagaraInt32, defaultValue=1}
 ]
}
